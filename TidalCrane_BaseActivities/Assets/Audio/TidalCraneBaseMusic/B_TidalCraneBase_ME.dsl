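; Music encounter for Tidal Crane base activities, playing Defense or Reclamation music by base type once event state allows it.
; Volume entry starts music at the owning actor, with attenuation minimum distance set by base name; exit pauses it.
; Base updates pass win, loss or defense stingers and a phase to audio parameters, then uninitialize when audio finishes.
; Reclamation completion also triggers the ending; defense updates only contain empty checks.
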(event EventInitializeMusicEncounter
  (|CallParentFunction)
  (bind _asb_tidal_crane_base_me_component (Utilities|Casting|CastToB_TidalCraneBase_ME_Component (Utilities|Array|Get(acopy) (MusicEncounter|GetMusicEncounterComponents))))
  (bind _output_get (Variables|Default|SetAs%20Music%20Encounter%20Component _asb_tidal_crane_base_me_component))
  (Variables|Default|SetOwningActor (Components|GetOwner _output_get)))

(event EventStartMusicEncounter
  (|CallParentFunction)
  (Variables|Default|GetReadyToInit
    (:then
      (Variables|Default|GetActiveMusicComponent
        (:then
          _)
        (:else
          (bind _audio_component (MusicEncounter|SpawnMusic2D (Utilities|Select :Index (Variables|Default|GetBase%20Type) :NewEnumerator4 (Variables|Sources|GetMSS_DefenseSource) :NewEnumerator0 (Variables|Sources|GetMSS_DefenseSource) :NewEnumerator1 (Variables|Sources|GetMSS_DefenseSource) :NewEnumerator2 (Variables|Sources|GetMSS_ReclamationSource) :NewEnumerator3 (Variables|Sources|GetMSS_ReclamationSource))))
          (Variables|Default|SetActiveMusicComponent _audio_component)))
      (bind _owningactor (Variables|Default|GetOwningActor)
        (:then
          (Transformation|SetWorldLocation (Variables|Default|GetActiveMusicComponent) (Transformation|GetActorLocation _owningactor) false false)
          (CallFunction|SetMinAttenuationValue 5e+03 8e+03 5e+03))
        (:else)))
    (:else)))

(event MusicEncounter|Volume|EventListenerEntered
  (|CallParentFunction)
  (bind _as_music_encounter_component (Variables|Default|GetAs%20Music%20Encounter%20Component)
    (:then
      (Class|BTidalCraneBaseMEComponent|GetBaseDetails _as_music_encounter_component)
      (CallFunction|StartMusicEncounter))
    (:else)))

(event MusicEncounter|Volume|EventListenerExited
  (|CallParentFunction)
  (CallFunction|PauseMusicEncounter))

(event Custom|BaseEventCompleted
  (bind _activemusiccomponent (Variables|Default|GetActiveMusicComponent)
    (:then
      (EventDispatchers|AssignOnAudioFinished _activemusiccomponent (delegate Custom|OnAudioFinished_Event_0)))
    (:else)))

(event Custom|OnAudioFinished_Event_0
  (CallFunction|UninitializeMusicEncounter))

(fn CreateBaseNames
  (return (Utilities|Array|MakeArray "Alpha" "Beta" "Charlie")))

(fn SetBaseType
  (params (BaseType :type TidalCrane_BaseTypes%20Enum :default "NewEnumerator4") (BaseNames :type String) (EventStatus :type TidalCrane_Bases_Enum%20Enum :default "NewEnumerator0"))
  (Variables|Default|SetBase%20Type BaseType)
  (Variables|Default|SetBaseName BaseNames)
  (bind _output_get (Variables|Default|SetEvent%20State EventStatus))
  (if (or (Utilities|Enum|Equal(Enum) _output_get "NewEnumerator2") (Utilities|Enum|Equal(Enum) _output_get "NewEnumerator4"))
    (Variables|Default|SetReadyToInit true)
    (if (not (MusicEncounter|GetMusicEncounterStateValue "NewEnumerator1"))
      (CallFunction|InitializeMusicEncounter)
      (return)
      (else
        (return)))
    (else
      (Variables|Default|SetReadyToInit false)
      (return))))

(fn SetMinAttenuationValue
  (params (Alpha :type Float%20%28single-precision%29) (Beta :type Float%20%28single-precision%29) (Charlie :type Float%20%28single-precision%29))
  (Variables|Default|SetAttnMin (select (Utilities|String|EqualExactly(String) (Variables|Default|GetBaseName) "Charlie") Charlie (select (Utilities|String|EqualExactly(String) (Variables|Default|GetBaseName) "Beta") Beta (select (Utilities|String|EqualExactly(String) (Variables|Default|GetBaseName) "Alpha") Alpha 4e+03))))
  (bind _activemusiccomponent (Variables|Default|GetActiveMusicComponent)
    (:then
      (Audio|Parameter|SetFloatParameter _activemusiccomponent "Attenuation.MinDistance" (Variables|Default|GetAttnMin) :!class /Script/AudioExtensions.AudioParameterControllerInterface))
    (:else)))

(fn UpdateBaseData
  (params (EventState :type TidalCrane_Bases_Enum%20Enum :default "NewEnumerator0") (AliveNPCs :type Integer) (MaxNPCs :type Integer) (BugsKilled :type Integer) (MaxBugs :type Integer) (CurTimerStartTime :type Float%20%28double-precision%29) (CurTimerDuration :type Float%20%28double-precision%29) (ReclamationPayloadSpawnersDestroyed :type Integer) (ReclamationPayloadMaxSpawners :type Integer))
  (Variables|Default|SetEvent%20State EventState)
  (switch Utilities|FlowControl|Switch|SwitchonTidalCrane_Bases_Enum EventState
    (:NewEnumerator5
      (LocalVariables|SetbDidWin true))
    (:NewEnumerator6
      (LocalVariables|SetbDidWin false)))
  (bind _activemusiccomponent (Variables|Default|GetActiveMusicComponent)
    (:then
      (Audio|Parameter|SetObjectArrayParameter _activemusiccomponent "Wavs.Stingers" (select (Utilities|Enum|Equal(Enum) (Variables|Default|GetBase%20Type) "NewEnumerator1") (LocalVariables|GetStingers_Defense) (select (LocalVariables|GetbDidWin) (LocalVariables|GetStingers_ReclamationWin) (LocalVariables|GetStingers_ReclamationLoss))))
      (Audio|Parameter|SetIntegerParameter _activemusiccomponent "MusicEncounter.Boss.Phase" 1 :!class /Script/AudioExtensions.AudioParameterControllerInterface)
      (CallFunction|BaseEventCompleted))
    (:else)))

(fn UpdateDefenseData
  (params (AliveNPCs :type Integer) (MaxNPCs :type Integer) (BugsKilled :type Integer) (MaxBugs :type Integer) (CurTimerStartTime :type Float%20%28double-precision%29) (CurTimerDuration :type Float%20%28double-precision%29))
  (Utilities|FlowControl|Sequence
    (:then_0
      (if (MusicEncounter|GetMusicEncounterStateValue "NewEnumerator2")
        _))
    (:then_1
      (if (== (Math|Integer|%25%28Integer%29 (/ (Variables|GetBugsKilled) (Variables|GetMaxBugs)) 2) 0)
        _))))

(fn UpdateReclamationData
  (params (AliveNPCs :type Integer) (MaxNPCs :type Integer) (BugsKilled :type Integer) (MaxBugs :type Integer) (SpawnersDestroyed :type Integer) (MaxSpawners :type Integer))
  (if (and (== (/ AliveNPCs MaxNPCs) 1.0) (== (/ SpawnersDestroyed MaxSpawners) 1.0))
    (bind _activemusiccomponent (Variables|Default|GetActiveMusicComponent)
      (:then
        (Audio|Parameter|SetIntegerParameter _activemusiccomponent "MusicEncounter.Boss.Phase" (Variables|Default|GetIndex_EventEnd) :!class /Script/AudioExtensions.AudioParameterControllerInterface)
        (CallFunction|BaseEventCompleted))
      (:else))))
