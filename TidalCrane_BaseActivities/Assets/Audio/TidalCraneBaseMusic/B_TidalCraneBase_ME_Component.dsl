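; Music encounter component that keeps a TidalCrane base's music encounter (B_TidalCraneBase_ME) in step with that base's type and event state.
; On initialize it caches the encounter and seeds it from the base locator.
; On BeginPlay, where cosmetics can run, it listens on the playspace event router for base type, Defense and Reclamation events.
; Events matching its base name push state and progress data, like NPC, bug, timer or spawner values, into the encounter.
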(event EventBeginPlay
  (|CallParentFunction)
  (Networking|CanExecuteCosmeticEvents
    (:True
      (bind _returnvalue (GameplayEventRouter|GetEventRouter "/Script/FortniteGame.FortPlayspace" (Components|GetOwner)))
      (Utilities|IsValid _returnvalue
        (:Is%20Valid
          (Utilities|FlowControl|Sequence
            (:then_0
              (bind (_asynctaskproxy _payload) (GameplayEventRouter|StartListeningToEvent _returnvalue "/TidalCrane_BaseActivities/Gameplay/TidalCrane_BaseType_Data.TidalCrane_BaseType_Data")
                (:OnEventReceived
                  (if (Utilities|String|EqualExactly(String) (Variables|Default|GetBaseName) (Variables|/TidalCrane_BaseActivities/Gameplay/B_TidalCrane_BaseLocator.B_TidalCrane_BaseLocator_C|GetBaseName (Utilities|Struct|BreakTidalCraneBaseTypeData _payload)))
                    (bind _as_b_tidal_crane_base_me (Variables|Default|GetAs%20B%20Tidal%20Crane%20Base%20ME)
                      (:then
                        (Class|BTidalCraneBaseME|SetBaseType _as_b_tidal_crane_base_me (.BaseType_6_4DA56F0F41FDC4B6AF1E939C29E8680E (Utilities|Struct|BreakTidalCraneBaseTypeData _payload)) (Variables|Default|GetBaseName) (Variables|/TidalCrane_BaseActivities/Gameplay/B_TidalCrane_BaseLocator.B_TidalCrane_BaseLocator_C|GetBaseState (Utilities|Struct|BreakTidalCraneBaseTypeData _payload))))
                      (:else))))))
            (:then_1
              ;[0.5751 0.1986 0.8854 0.5] Defense
              (bind (_asynctaskproxy_2 _payload_2) (GameplayEventRouter|StartListeningToEvent _returnvalue "/TidalCrane_BaseActivities/Gameplay/TidalCrane_Base_DefenseData.TidalCrane_Base_DefenseData")
                (:OnEventReceived
                  (if (Utilities|String|EqualExactly(String) (Variables|Default|GetBaseName) (Variables|/TidalCrane_BaseActivities/Gameplay/B_TidalCrane_BaseLocator.B_TidalCrane_BaseLocator_C|GetBaseName (Variables|/TidalCrane_BaseActivities/Gameplay/B_TidalCrane_MilitaryActivity_SpawnManager.B_TidalCrane_MilitaryActivity_SpawnManager_C|GetBaseLocator (Utilities|Struct|BreakTidalCraneBaseDefenseData _payload_2))))
                    (Variables|Default|SetEvent%20State (.EventState_10_5AACBE6344F0C1F9283B6F996EE80287 (Utilities|Struct|BreakTidalCraneBaseDefenseData _payload_2)))
                    (bind _as_b_tidal_crane_base_me_2 (Variables|Default|GetAs%20B%20Tidal%20Crane%20Base%20ME)
                      (:then
                        (bind _as_b_tidal_crane_base_me_3 (Variables|Default|GetAs%20B%20Tidal%20Crane%20Base%20ME)
                          (:then
                            (Class|BTidalCraneBaseME|SetBaseType _as_b_tidal_crane_base_me_3 (Variables|Default|GetBase%20Type) (Variables|Default|GetBaseName) (Variables|Default|GetEvent%20State))
                            (bind (_eventdata_spawnmanager_14_a13c79344bac9fd301fbf6be5b2811b6 _eventdata_eventstate_10_5aacbe6344f0c1f9283b6f996ee80287 _eventdata_alivenpcs_2_3c75bf2948d8cae5eae1b7a478022f31 _eventdata_maxnpcs_4_8ff348804ec5364cd593cfb7d51198f9 _eventdata_bugskilled_6_97a8a9e7411f76a1dadb4daa60527f22 _eventdata_maxbugs_27_e3a9792f47777d86969e85a4a891a09f _eventdata_currtimerstarttime_22_731403c943e3202038d02c8f1c0abfbc _eventdata_currtimerduration_24_c38d03a840c86e36c472af8673c34d7b) (Utilities|Struct|BreakTidalCraneBaseDefenseData (Variables|/TidalCrane_BaseActivities/Gameplay/DefenseGameplay/SpawnManager_TidalCrane_Defense.SpawnManager_TidalCrane_Defense_C|GetEventData_SpawnManager_14_A13C79344BAC9FD301FBF6BE5B2811B6 (Utilities|Struct|BreakTidalCraneBaseDefenseData _payload_2))))
                            (Class|BTidalCraneBaseME|UpdateBaseData _as_b_tidal_crane_base_me_2 _eventdata_eventstate_10_5aacbe6344f0c1f9283b6f996ee80287 _eventdata_alivenpcs_2_3c75bf2948d8cae5eae1b7a478022f31 _eventdata_maxnpcs_4_8ff348804ec5364cd593cfb7d51198f9 _eventdata_bugskilled_6_97a8a9e7411f76a1dadb4daa60527f22 _eventdata_maxbugs_27_e3a9792f47777d86969e85a4a891a09f _eventdata_currtimerstarttime_22_731403c943e3202038d02c8f1c0abfbc _eventdata_currtimerduration_24_c38d03a840c86e36c472af8673c34d7b 0 0))
                          (:else)))
                      (:else))))))
            (:then_2
              (bind (_asynctaskproxy_3 _payload_3) (GameplayEventRouter|StartListeningToEvent _returnvalue "/TidalCrane_BaseActivities/Gameplay/TidalCrane_Base_Reclamation_Data.TidalCrane_Base_Reclamation_Data")
                (:OnEventReceived
                  ;[0.5751 0.1986 0.8854 0.5] Reclamation
                  (if (Utilities|String|EqualExactly(String) (Variables|Default|GetBaseName) (Variables|/TidalCrane_BaseActivities/Gameplay/B_TidalCrane_BaseLocator.B_TidalCrane_BaseLocator_C|GetBaseName (Variables|/TidalCrane_BaseActivities/Gameplay/B_TidalCrane_MilitaryActivity_SpawnManager.B_TidalCrane_MilitaryActivity_SpawnManager_C|GetBaseLocator (Utilities|Struct|BreakTidalCraneBaseReclamationData _payload_3))))
                    (Variables|Default|SetEvent%20State (.EventState_5_781D3B7C4B9A8EFDCFC0F5832D1CE7C1 (Utilities|Struct|BreakTidalCraneBaseReclamationData _payload_3)))
                    (bind _as_b_tidal_crane_base_me_4 (Variables|Default|GetAs%20B%20Tidal%20Crane%20Base%20ME)
                      (:then
                        (Class|BTidalCraneBaseME|SetBaseType _as_b_tidal_crane_base_me_4 (Variables|Default|GetBase%20Type) (Variables|Default|GetBaseName) (Variables|Default|GetEvent%20State))
                        (Utilities|IsValid (Utilities|Struct|BreakTidalCraneBaseReclamationData _payload_3)
                          (:Is%20Valid
                            (bind (_eventdata_spawnmanager_2_73860b3c407b1bc6e2a4bbbc629f37fa _eventdata_eventstate_5_781d3b7c4b9a8efdcfc0f5832d1ce7c1 _eventdata_freednpcs_8_ba4606824ff6aa73adbf20ac85ab4e02 _eventdata_maxnpcstofree_17_feb9e4fd4cd339ff26e5e7af3e52f04c _eventdata_bugskilled_10_fcec42d04ca2b85d8e9b6eb4ce1804da _eventdata_spawnersdestroyed_12_94bdde484207b110c717a5b7340381ff _eventdata_maxspawners_14_c069eae442d53210b1269db2a5ba1ef5) (Utilities|Struct|BreakTidalCraneBaseReclamationData (Variables|/TidalCrane_BaseActivities/Gameplay/ReclaimGameplay/SpawnManager_TidalCrane_Reclaim.SpawnManager_TidalCrane_Reclaim_C|GetEventData_SpawnManager_2_73860B3C407B1BC6E2A4BBBC629F37FA (Utilities|Struct|BreakTidalCraneBaseReclamationData _payload_3))))
                            (Class|BTidalCraneBaseME|UpdateBaseData _as_b_tidal_crane_base_me_4 _eventdata_eventstate_5_781d3b7c4b9a8efdcfc0f5832d1ce7c1 _eventdata_freednpcs_8_ba4606824ff6aa73adbf20ac85ab4e02 _eventdata_maxnpcstofree_17_feb9e4fd4cd339ff26e5e7af3e52f04c _eventdata_bugskilled_10_fcec42d04ca2b85d8e9b6eb4ce1804da 0 0.0 0.0 _eventdata_spawnersdestroyed_12_94bdde484207b110c717a5b7340381ff _eventdata_maxspawners_14_c069eae442d53210b1269db2a5ba1ef5))))
                      (:else))))))))))))

(event EventInitializeComponent
  (|CallParentFunction)
  (bind _asb_tidal_crane_base_me (Utilities|Casting|CastToB_TidalCraneBase_ME (MusicEncounterComponent|GetMusicEncounter)))
  (Variables|Default|SetAs%20B%20Tidal%20Crane%20Base%20ME _asb_tidal_crane_base_me)
  (bind _baselocator (Variables|Default|GetBaseLocator)
    (:then
      (Variables|Default|SetEvent%20State (Variables|/TidalCrane_BaseActivities/Gameplay/B_TidalCrane_BaseLocator.B_TidalCrane_BaseLocator_C|GetBaseState _baselocator))
      (Variables|Default|SetBase%20Type (Variables|/TidalCrane_BaseActivities/Gameplay/B_TidalCrane_BaseLocator.B_TidalCrane_BaseLocator_C|GetBaseType _baselocator))
      (bind _as_b_tidal_crane_base_me (Variables|Default|GetAs%20B%20Tidal%20Crane%20Base%20ME)
        (:then
          (Class|BTidalCraneBaseME|SetBaseType _as_b_tidal_crane_base_me (Variables|/TidalCrane_BaseActivities/Gameplay/B_TidalCrane_BaseLocator.B_TidalCrane_BaseLocator_C|GetBaseType _baselocator) (Variables|Default|GetBaseName) (Variables|/TidalCrane_BaseActivities/Gameplay/B_TidalCrane_BaseLocator.B_TidalCrane_BaseLocator_C|GetBaseState _baselocator)))
        (:else)))
    (:else)))

(fn ChooseBaseName
  (return (Utilities|Array|MakeArray "Alpha" "Beta" "Charlie")))

(fn GetBaseDetails
  (bind _baselocator (Variables|Default|GetBaseLocator)
    (:then
      (if (Utilities|String|EqualExactly(String) (Variables|Default|GetBaseName) (Variables|/TidalCrane_BaseActivities/Gameplay/B_TidalCrane_BaseLocator.B_TidalCrane_BaseLocator_C|GetBaseName _baselocator))
        (Variables|Default|SetEvent%20State (Variables|/TidalCrane_BaseActivities/Gameplay/B_TidalCrane_BaseLocator.B_TidalCrane_BaseLocator_C|GetBaseState _baselocator))
        (Variables|Default|SetBase%20Type (Variables|/TidalCrane_BaseActivities/Gameplay/B_TidalCrane_BaseLocator.B_TidalCrane_BaseLocator_C|GetBaseType _baselocator))
        (Class|BTidalCraneBaseME|SetBaseType (Variables|Default|GetAs%20B%20Tidal%20Crane%20Base%20ME) (Variables|Default|GetBase%20Type) (Variables|Default|GetBaseName) (Variables|Default|GetEvent%20State))))
    (:else)))

(fn SetBaseType
  (params (BaseType :type TidalCrane_BaseTypes%20Enum :default "NewEnumerator4") (BaseNames :type String) (EventStatus :type TidalCrane_Bases_Enum%20Enum :default "NewEnumerator0"))
  (if (Utilities|Enum|NotEqual(Enum) BaseType "NewEnumerator4")
    (Variables|Default|SetBase%20Type (Variables|GetBaseType))
    (bind _as_b_tidal_crane_base_me (Variables|Default|GetAs%20B%20Tidal%20Crane%20Base%20ME)
      (:then
        (Class|BTidalCraneBaseME|SetBaseType _as_b_tidal_crane_base_me (Variables|GetBaseType) BaseNames EventStatus))
      (:else))))
