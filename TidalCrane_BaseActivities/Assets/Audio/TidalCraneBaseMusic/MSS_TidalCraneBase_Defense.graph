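# Stereo MetaSound music source for the TidalCrane base Defense encounter, layering looping percussion and other stems with phase stingers.
# A boss phase change fires a tempo-quantized stinger, which releases the loop envelope and finishes the sound when done.
# Distance curves shape levels, weapon-fire modulation ducks loops, and an SFX-bus envelope follower drives a percussion EQ sidechain.
# Pausing fades out and ends playback; an alarm control-bus modulator slowly lowers the right channel only.
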MetaSoundSource /TidalCrane_BaseActivities/Assets/Audio/TidalCraneBaseMusic/MSS_TidalCraneBase_Defense
inputs [
 Attenuation.MaxDistance {type=Float}
 Attenuation.MinDistance {type=Float}
 Gain.Stingers {type=Float}
 MusicEncounter.bIsPaused {type=Bool}
 MusicEncounter.Boss.Phase {type=Int32}
 MusicEncounter.PauseFadeOutTime {type=Time}
 Sidechain.Percs.AudioBus {type=AudioBusAsset, access=Constructor}
 Sidechain.Percs.Bandwidth {type=Float, access=Constructor}
 Sidechain.Percs.Freq {type=Float, access=Constructor}
 Sidechain.Percs.Gain.Max {type=Float, access=Constructor}
 Sidechain.Percs.Gain.Min {type=Float, access=Constructor}
 Tempo {type=Float, access=Constructor}
 UE.Attenuation.Distance {type=Float}
 UE.Source.OnPlay {type=Trigger}
 Wavs.Loops.Else {type=WaveAsset:Array, access=Constructor}
 Wavs.Loops.Percs {type=WaveAsset:Array, access=Constructor}
 Wavs.Stingers {type=WaveAsset:Array}
 Wavs.SWITCHONLY {type=WaveAsset:Array, access=Constructor}
]
interfaces [UE.Attenuation:1.0, UE.OutputFormat.Stereo:1.0, UE.Source.OneShot:1.0, UE.Source:1.1]
outputs [
 Envelope {type=Float}
 "UE.OutputFormat.Stereo.Audio:0" {type=Audio}
 "UE.OutputFormat.Stereo.Audio:1" {type=Audio}
 UE.Source.OneShot.OnFinished {type=Trigger}
]
pages [Default {
  input_defaults {
   Attenuation.MaxDistance=16000
   Attenuation.MinDistance=4000
   Gain.Stingers=2.5119
   MusicEncounter.Boss.Phase=0
   MusicEncounter.PauseFadeOutTime=3
   MusicEncounter.bIsPaused=null
   Sidechain.Percs.AudioBus=/TidalCrane_BaseActivities/Assets/Audio/TidalCraneMixing/AB_TidalCraneBase_SFX_Sidechain
   Sidechain.Percs.Bandwidth=1
   Sidechain.Percs.Freq=200
   Sidechain.Percs.Gain.Max=-3
   Sidechain.Percs.Gain.Min=0
   Tempo=146
   UE.Attenuation.Distance=16000
   UE.Source.OnPlay=null
   Wavs.SWITCHONLY=null
   Wavs.Loops.Else [&/MX_S37_Combat_TidalCraneBase_Defense_Phase_1_LP_Else]
   Wavs.Loops.Percs [null]
   Wavs.Stingers [null, &/MX_S37_Combat_TidalCraneBase_Defense_Complete_STNG]
  }
  nodes [
   n1 = HarmonixNodes.TriggerToTransport(Play=^UE.Source.OnPlay)
   TriggerOnValueChange_2 = UE.Trigger On Value Change:Int32(Value=^MusicEncounter.Boss.Phase)
   TriggerAnyTriggerAny2_0 = "TriggerAny.Trigger Any (2)"(In 1=^UE.Source.OnPlay, In 0=@TriggerOnValueChange_2.Trigger)
   TriggerOnce_0 = UE.Trigger Once(
    Trigger In=HarmonixNodes.MidiClockSubdivisionTrigger(MIDI Clock=HarmonixNodes.Metronome(Tempo=^Tempo, Transport=@n1.Transport).MIDI Clock).Trigger Out
    Reset=UE.Reroute(Value=@TriggerAnyTriggerAny2_0.Out).Value
   ) {also {defaults {Start Closed="<MetaSoundBuilderResult.SUCCEEDED: 0>"}}}
   n2 = UE.Reroute(Value=@TriggerOnce_0.Trigger Out)
   TriggerDelay_1 = UE.Trigger Delay(Delay Time=UE.BPMToSeconds(BPM=^Tempo).Seconds, In=@n2.Value)
   TriggerOnce_1 = UE.Trigger Once(Trigger In=@n2.Value, Reset=@TriggerDelay_1.Out)
   n3 = UE.Reroute(Value=TriggerCompare.Clamp:Int32(A=^MusicEncounter.Boss.Phase, Compare=@TriggerOnce_1.Trigger Out) {also {defaults {B=1}}}.True)
   WavePlayer_2 = UE.Wave Player:Stereo(Wave Asset=Array.Get:WaveAsset:Array(Index=^MusicEncounter.Boss.Phase, Array=^Wavs.Stingers, Trigger=@n3.Value).Element, Play=@n3.Value) {also {version [1,1]}}
   AudioBusWriter2_0 = "UE.Audio Bus Writer (2)"(In 0=@WavePlayer_2.Out Left, In 1=@WavePlayer_2.Out Right) {also {defaults {Audio Bus=&AB_Sidechain_FortCreatures_Vol}}}
   BiquadFilter_0 = UE.Biquad Filter:Audio() {also {defaults {Type=4}}}
   InitVariable.Trigger()
   n4 = UE.BPMToSeconds(BPM=^Tempo)
   ADSREnvelopeADSREnvelope_0 = ADSR Envelope.ADSR Envelope:Float(Trigger Attack=^UE.Source.OnPlay, Attack Time=@n4.Seconds, Release Time=@n4.Seconds, Trigger Release=@WavePlayer_2.On Play) {also {defaults {Sustain Level=1}}}
   AudioBusReader1_0 = "UE.Audio Bus Reader (1)"(Audio Bus=^Sidechain.Percs.AudioBus)
   EnvelopeFollower_0 = UE.Envelope Follower(In=@AudioBusReader1_0.Out 0)
   n5 = MapRange.MapRange:Float(Out Range A=^Sidechain.Percs.Gain.Max, Out Range B=^Sidechain.Percs.Gain.Min, In=@EnvelopeFollower_0.Envelope)
   n6 = UE.Reroute(Value=^UE.Source.OnPlay)
   n7 = UE.Reroute(Value=@n6.Value)
   n8 = VariableDeferredAccessor.Trigger(Variable=InitVariable.Trigger().Variable)
   WavePlayer_3 = UE.Wave Player:Stereo(
    Wave Asset=Array.Get:WaveAsset:Array(Index=^MusicEncounter.Boss.Phase, Array=^Wavs.Loops.Percs, Trigger=@n7.Value).Element
    Play=@n7.Value
    Stop=@n8.Value
   ) {also {defaults {Loop="<MetaSoundBuilderResult.SUCCEEDED: 0>"}, version [1,1]}}
   BiquadFilter_1 = UE.Biquad Filter:Audio(Bandwidth=^Sidechain.Percs.Bandwidth, Cutoff Frequency=^Sidechain.Percs.Freq, Gain=@n5.Out Value, In=@WavePlayer_3.Out Right) {also {defaults {Type=4}}}
   BiquadFilter_2 = UE.Biquad Filter:Audio(Bandwidth=^Sidechain.Percs.Bandwidth, Cutoff Frequency=^Sidechain.Percs.Freq, Gain=@n5.Out Value, In=@WavePlayer_3.Out Left) {also {defaults {Type=4}}}
   n9 = Modulation.GetModulatorValue() {also {defaults {Modulator=/Game/Sounds/Modulation/EnvelopeFollowers/FortWeapon/CB_WeaponFireEnvFollower}}}
   DecibelstoLinearGain_1 = UE.Decibels to Linear Gain:Float(Decibels=MapRange.MapRange:Float(In=@n9.Out) {also {defaults {In Range A=1, In Range B=0, Out Range A=-6, Out Range B=0}}}.Out Value)
   n10 = Clamp.Clamp:Float(In=UE.WaveTableBankEvaluate(Input=MapRange.MapRange:Float(In Range B=^Attenuation.MaxDistance, In Range A=^Attenuation.MinDistance, In=^UE.Attenuation.Distance).Out Value) {also {defaults {Index=1, WaveTableBank=/TidalCrane_BaseActivities/Assets/Audio/TidalCraneMixing/WTB_TidalCraneBase_Defense}}}.Output) {also {defaults {Max=1}}}
   DecibelstoLinearGain_0 = UE.Decibels to Linear Gain:Float(Decibels=MapRange.MapRange:Float(In=@n9.Out) {also {defaults {In Range A=1, In Range B=0, Out Range A=-3, Out Range B=0}}}.Out Value)
   n11 = UE.Reroute(Value=@n6.Value)
   WavePlayer_1 = UE.Wave Player:Stereo(
    Wave Asset=Array.Get:WaveAsset:Array(Index=^MusicEncounter.Boss.Phase, Array=^Wavs.Loops.Else, Trigger=@n11.Value).Element
    Play=@n11.Value
    Stop=@n8.Value
   ) {also {defaults {Loop="<MetaSoundBuilderResult.SUCCEEDED: 0>"}, version [1,1]}}
   AudioMixerAudioMixerStereo2_0 = "AudioMixer.Audio Mixer (Stereo, 2)"(
    In 1 R=@BiquadFilter_1.Out
    In 1 L=@BiquadFilter_2.Out
    Gain 1=UE.Multiply:Float(
     PrimaryOperand=@DecibelstoLinearGain_1.Linear Gain
     AdditionalOperands=UE.Multiply:Float(PrimaryOperand=@n10.Value) {also {defaults {AdditionalOperands=2}}}.Out
    ).Out
    Gain 0=UE.Multiply:Float(
     PrimaryOperand=@DecibelstoLinearGain_0.Linear Gain
     AdditionalOperands=UE.Multiply:Float(PrimaryOperand=@n10.Value) {also {defaults {AdditionalOperands=2}}}.Out
    ).Out
    In 0 L=@WavePlayer_1.Out Left
    In 0 R=@WavePlayer_1.Out Right
   )
   n12 = UE.Multiply:Float(PrimaryOperand=UE.WaveTableBankEvaluate(Input=MapRange.MapRange:Float(In Range B=^Attenuation.MaxDistance, In Range A=^Attenuation.MinDistance, In=^UE.Attenuation.Distance).Out Value) {also {defaults {WaveTableBank=&WTB_TidalCraneBase_Defense}}}.Output)
   WavePlayer_0 = UE.Wave Player:Stereo(Play=^UE.Source.OnPlay, Stop=VariableDeferredAccessor.Trigger(Variable=InitVariable.Trigger().Variable).Value) {also {
     defaults {Loop="<MetaSoundBuilderResult.SUCCEEDED: 0>", Wave Asset=/ProtoSwarm/Assets/Audio/SoundWaves/Music/MX_S37_Combat_ArmoredBug_Phase_0_LP_Else}
     version [1,1]
    }}
   AudioMixerAudioMixerStereo3_0 = "AudioMixer.Audio Mixer (Stereo, 3)"(
    Gain 1=@ADSREnvelopeADSREnvelope_0.Out Envelope
    Gain 2=@ADSREnvelopeADSREnvelope_0.Out Envelope
    In 1 L=@AudioMixerAudioMixerStereo2_0.Out L
    In 1 R=@AudioMixerAudioMixerStereo2_0.Out R
    In 2 L=UE.Multiply:Audio by Float(AdditionalOperands=@n12.Out, PrimaryOperand=@WavePlayer_0.Out Left).Out
    In 2 R=UE.Multiply:Audio by Float(AdditionalOperands=@n12.Out, PrimaryOperand=@WavePlayer_0.Out Right).Out
    Gain 0=UE.Multiply:Float(
     AdditionalOperands=^Gain.Stingers
     PrimaryOperand=MapRange.MapRange:Float(In Range B=^Attenuation.MaxDistance, In Range A=^Attenuation.MinDistance, In=^UE.Attenuation.Distance) {also {defaults {Out Range A=1, Out Range B=0.1}}}.Out Value
    ).Out
    In 0 L=@WavePlayer_2.Out Left
    In 0 R=@WavePlayer_2.Out Right
   )
   TriggerOnValueChange_0 = UE.Trigger On Value Change:Bool(Value=^MusicEncounter.bIsPaused)
   n13 = TriggerCompare.Clamp:Bool(A=^MusicEncounter.bIsPaused, Compare=@TriggerOnValueChange_0.Trigger) {also {defaults {B="<MetaSoundBuilderResult.SUCCEEDED: 0>"}}}
   n14 = Value.Value:Float(Reset=@n13.False, Set=@n13.True) {also {defaults {Init Value=1}}}
   n15 = UE.InterpTo:Audio(Interp Time=^MusicEncounter.PauseFadeOutTime, Target=@n14.Output Value)
   Multiply_2 = UE.Multiply:Audio by Float(PrimaryOperand=@AudioMixerAudioMixerStereo3_0.Out L, AdditionalOperands=@n15.Value)
   DecibelstoLinearGain_2 = UE.Decibels to Linear Gain:Float() {also {defaults {Decibels=-4}}}
   n16 = Modulation.GetModulatorValue() {also {defaults {Modulator=/TidalCrane_BaseActivities/Assets/Audio/TidalCraneBaseSpeakers/CB_TidalCraneBase_Defense_AlarmDiegesisControl}}}
   TriggerOnValueChange_1 = UE.Trigger On Value Change:Float(Value=@n16.Out)
   TriggerRouteTriggerRouteFloat2_0 = "TriggerRoute.Trigger Route (Float, 2):Float"(
    Set 0=^UE.Source.OnPlay
    Value 1=@DecibelstoLinearGain_2.Linear Gain
    Set 1=TriggerCompare.Clamp:Float(A=@n16.Out, Compare=@TriggerOnValueChange_1.Trigger).True
   ) {also {defaults {Value 0=1}}}
   Multiply_3 = UE.Multiply:Audio by Float(
    AdditionalOperands=UE.InterpTo:Audio(Target=@TriggerRouteTriggerRouteFloat2_0.Value) {also {defaults {Interp Time=60}}}.Value
    PrimaryOperand=UE.Multiply:Audio by Float(PrimaryOperand=@AudioMixerAudioMixerStereo3_0.Out R, AdditionalOperands=@n15.Value).Out
   )
   UE.Multiply:Float(PrimaryOperand=Modulation.GetModulatorValue() {also {defaults {Modulator=&CB_WeaponFireEnvFollower}}}.Out) {also {defaults {AdditionalOperands=2}}}
   TriggerDelay_0 = UE.Trigger Delay(Delay Time=^MusicEncounter.PauseFadeOutTime, Reset=@n14.On Reset, In=@n14.On Set)
   TriggerAnyTriggerAny2_1 = "TriggerAny.Trigger Any (2)"(In 1=@TriggerDelay_0.Out, In 0=@WavePlayer_2.On Finished)
   n17 = Array.Num:WaveAsset:Array(Array=^Wavs.Stingers)
   VariableMutator.Int32(
    Variable=InitVariable.Int32().Variable
    Value=Value.Value:Int32(Set=^UE.Source.OnPlay, Target Value=UE.Subtract:Int32(PrimaryOperand=@n17.Num) {also {defaults {AdditionalOperands=1}}}.Out).Output Value
   )
  ]
  links [
   {dst="@UE.OutputFormat.Stereo.Audio:0", src=Multiply_2.Out}
   {dst="@UE.OutputFormat.Stereo.Audio:1", src=Multiply_3.Out}
   {dst="@UE.Source.OneShot.OnFinished", src=TriggerAnyTriggerAny2_1.Out}
  ]
 }]
properties {
 assetImportData=&self:AssetImportData
 numChannels=2
 outputFormat=Stereo
 sampleRate=48000
 soundClassObject=&Fort_Music_NPC
 virtualizationMode=PlayWhenSilent
 volume=0.38
 busSends [{audioBus=&AB_Sidechain_FortAmbient_Vol, sendLevel=1}]
}
