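# Stereo MetaSound music source for the TidalCrane base Reclamation activity, playing the exterior loop.
# Listener distance drives a wavetable curve and a linear ramp, scaled by Idle and Stingers gains, atop a fixed attenuation.
# A pause or qualifying boss-phase change fades the music and fires OnFinished after the fade time; unpausing resets both.
# Percussion, sidechain, tempo and wave-array inputs are declared but unconnected.
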MetaSoundSource /TidalCrane_BaseActivities/Assets/Audio/TidalCraneBaseMusic/MSS_TidalCraneBase_Reclamation
inputs [
 Attenuation.MaxDistance {type=Float}
 Attenuation.MinDistance {type=Float}
 Gain.Idle {type=Float}
 Gain.Perc {type=Float}
 Gain.Stingers {type=Float}
 In 1 {type=Trigger}
 MusicEncounter.bIsPaused {type=Bool}
 MusicEncounter.Boss.Phase {type=Int32}
 MusicEncounter.PauseFadeOutTime {type=Time}
 Sidechain.Percs.AudioBus {type=AudioBusAsset, access=Constructor}
 Sidechain.Percs.Bandwidth {type=Float, access=Constructor}
 Sidechain.Percs.Freq {type=Float, access=Constructor}
 Sidechain.Percs.Gain.Max {type=Float, access=Constructor}
 Sidechain.Percs.Gain.Min {type=Float, access=Constructor}
 Tempo {type=Float, access=Constructor}
 UE.Attenuation.Distance {type=Float}
 UE.Source.OnPlay {type=Trigger}
 Wavs.Loops.Else {type=WaveAsset:Array, access=Constructor}
 Wavs.Loops.Percs {type=WaveAsset:Array, access=Constructor}
 Wavs.Stingers {type=WaveAsset:Array}
 Wavs.SWITCHONLY {type=WaveAsset:Array, access=Constructor}
]
interfaces [UE.Attenuation:1.0, UE.OutputFormat.Stereo:1.0, UE.Source.OneShot:1.0, UE.Source:1.1]
outputs [
 "UE.OutputFormat.Stereo.Audio:0" {type=Audio}
 "UE.OutputFormat.Stereo.Audio:1" {type=Audio}
 UE.Source.OneShot.OnFinished {type=Trigger}
]
pages [Default {
  input_defaults {
   Attenuation.MaxDistance=17000
   Attenuation.MinDistance=4000
   Gain.Idle=3.9811
   Gain.Perc=1
   Gain.Stingers=1.9953
   In 1=null
   MusicEncounter.Boss.Phase=0
   MusicEncounter.PauseFadeOutTime=3
   MusicEncounter.bIsPaused=null
   Sidechain.Percs.AudioBus=/TidalCrane_BaseActivities/Assets/Audio/TidalCraneMixing/AB_TidalCraneBase_SFX_Sidechain
   Sidechain.Percs.Bandwidth=1
   Sidechain.Percs.Freq=200
   Sidechain.Percs.Gain.Max=-3
   Sidechain.Percs.Gain.Min=0
   Tempo=145
   UE.Attenuation.Distance=4000
   UE.Source.OnPlay=null
   Wavs.Loops.Else [&/MX_S37_POI_BaseReclamation_Exterior_LP]
   Wavs.Loops.Percs []
   Wavs.SWITCHONLY []
   Wavs.Stingers []
  }
  nodes [
   InitVariable.Int32()
   InitVariable.Trigger()
   InitVariable.Trigger()
   n1 = UE.Multiply:Float(
    AdditionalOperands=^Gain.Idle
    PrimaryOperand=UE.WaveTableBankEvaluate(Input=MapRange.MapRange:Float(In Range B=^Attenuation.MaxDistance, In Range A=^Attenuation.MinDistance, In=^UE.Attenuation.Distance).Out Value) {also {defaults {Index=3, WaveTableBank=/TidalCrane_BaseActivities/Assets/Audio/TidalCraneMixing/WTB_TidalCraneBase_Defense}}}.Output
   )
   WavePlayer_0 = UE.Wave Player:Stereo(Play=^UE.Source.OnPlay, Stop=VariableDeferredAccessor.Trigger(Variable=InitVariable.Trigger().Variable).Value) {also {
     defaults {
      Loop="<MetaSoundBuilderResult.SUCCEEDED: 0>"
      Wave Asset=&MX_S37_POI_BaseReclamation_Exterior_LP
     }
     version [1,1]
    }}
   n2 = UE.Multiply:Float(
    AdditionalOperands=^Gain.Stingers
    PrimaryOperand=MapRange.MapRange:Float(In Range B=^Attenuation.MaxDistance, In Range A=^Attenuation.MinDistance, In=^UE.Attenuation.Distance) {also {defaults {Out Range A=1, Out Range B=0.1}}}.Out Value
   )
   DecibelstoLinearGain_0 = UE.Decibels to Linear Gain:Float() {also {defaults {Decibels=-19}}}
   TriggerOnValueChange_0 = UE.Trigger On Value Change:Bool(Value=^MusicEncounter.bIsPaused)
   n3 = TriggerCompare.Clamp:Bool(A=^MusicEncounter.bIsPaused, Compare=@TriggerOnValueChange_0.Trigger) {also {defaults {B="<MetaSoundBuilderResult.SUCCEEDED: 0>"}}}
   TriggerOnValueChange_1 = UE.Trigger On Value Change:Int32(Value=^MusicEncounter.Boss.Phase)
   TriggerAnyTriggerAny2_0 = "TriggerAny.Trigger Any (2)"(
    In 1=@n3.True
    In 0=TriggerCompare.Clamp:Int32(A=UE.Reroute(Value=^MusicEncounter.Boss.Phase).Value, Compare=@TriggerOnValueChange_1.Trigger) {also {defaults {Type=1}}}.True
   )
   n4 = Value.Value:Float(Set=@TriggerAnyTriggerAny2_0.Out, Reset=@n3.False) {also {defaults {Init Value=1}}}
   n5 = UE.Multiply:Float(
    AdditionalOperands=@DecibelstoLinearGain_0.Linear Gain
    PrimaryOperand=UE.InterpTo:Audio(Interp Time=^MusicEncounter.PauseFadeOutTime, Target=@n4.Output Value).Value
   )
   Multiply_0 = UE.Multiply:Audio by Float(
    PrimaryOperand=UE.Multiply:Audio by Float(
     PrimaryOperand=UE.Multiply:Audio by Float(AdditionalOperands=@n1.Out, PrimaryOperand=@WavePlayer_0.Out Right).Out
     AdditionalOperands=@n2.Out
    ).Out
    AdditionalOperands=@n5.Out
   )
   Multiply_3 = UE.Multiply:Audio by Float(
    PrimaryOperand=UE.Multiply:Audio by Float(PrimaryOperand=UE.Multiply:Audio by Float(AdditionalOperands=@n1.Out, PrimaryOperand=@WavePlayer_0.Out Left).Out, AdditionalOperands=@n2.Out).Out
    AdditionalOperands=@n5.Out
   )
   TriggerDelay_0 = UE.Trigger Delay(Delay Time=^MusicEncounter.PauseFadeOutTime, Reset=@n4.On Reset, In=@n4.On Set)
  ]
  links [
   {dst="@UE.OutputFormat.Stereo.Audio:1", src=Multiply_0.Out}
   {dst="@UE.OutputFormat.Stereo.Audio:0", src=Multiply_3.Out}
   {dst="@UE.Source.OneShot.OnFinished", src=TriggerDelay_0.Out}
  ]
 }]
properties {
 assetImportData=&self:AssetImportData
 numChannels=2
 outputFormat=Stereo
 sampleRate=48000
 soundClassObject=&Fort_Music_NPC
 virtualizationMode=PlayWhenSilent
 volume=2
}
